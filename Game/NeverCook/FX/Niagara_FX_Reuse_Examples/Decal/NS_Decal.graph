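# Niagara example system that spawns a short-lived, tinted ground decal through a single burst emitter, GroundDecal, and a decal renderer.
# Spawn orients the decal along an up vector and sizes it with a front-end modifier that picks a smaller scale outside gameplay.
# Update curves ramp decal fade in and out, fade alpha late in life, and darken RGB, with brightness scaled by time of day, dimmest at night.
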NiagaraSystem /Game/NeverCook/FX/Niagara_FX_Reuse_Examples/Decal/NS_Decal
dependencies {usedDataInterfaces [&NiagaraDataInterfaceCurve], usedDynamicInputs [], usedModules [&SystemState]}
emitters {GroundDecal {
  emitterUpdateScript [
   &EmitterState(Life Cycle Mode=::NewEnumerator1, Loop Behavior=::NewEnumerator1, Loop Delay=0.05, Loop Duration=1.5, UseLoopDelay=false)
   &SpawnBurst_Instantaneous(Use Spawn Probability=false, UseLoopCountLimit=false)
  ]
  particleSpawnScript [
   &InitializeParticle(
    Color=<rgba 0.3 0.261 0.21 1>
    Lifetime=1.5
    Position Offset Coordinate Space=::Local
    Sprite Size Mode=::NewEnumerator3
    Uniform Sprite Size=200
    UsePositionOffset=true
    Write Material Random=true
   )
   /Niagara/Modules/Spawn/Decal/InitialDecalOrientation(Orientation Vector:Vector3f=<0 0 1>, Use Rotation Vector=false)
   /SetVariables(Particles.DecalSizeScale:Vector3f=&Multiply_VectorByFloat(Vector=<1 1 1>, Float=&FN_FrontEndFloatModifier(Gameplay=250, Front End=150))) {isSetParameters=true}
  ]
  particleUpdateScript [
   /SetVariables(Particles.DecalFade:NiagaraFloat=&FloatFromCurve(FloatCurve=DataInterface() {propertyValues {Curve {keys [
         <at 0 0>
         <at 0.05 0.7>
         {arriveTangent=-0.7368, interpMode=RCIM_Cubic, leaveTangent=-0.7368, time=0.8, value=1}
         {arriveTangent=0.4472, interpMode=RCIM_Cubic, leaveTangent=0.4472, tangentMode=RCTM_User, time=1, value=0}
        ]}}})) {isSetParameters=true}
   &ParticleState()
   &ScaleColor(
    Scale Alpha=&FloatFromCurve(FloatCurve=DataInterface() {propertyValues {Curve {keys [<at 0 1>,<at 0.7 1>,<at 1 0>]}}})
    Scale RGB=&Multiply_VectorByFloat(
     Vector=<1 1 1>
     Float=&FloatFromCurve(
      FloatCurve=DataInterface() {propertyValues {Curve {keys [<at 0 1>,<at 1 0.5>]}}}
      Scale Curve=&FNTimeOfDayScalar(Night Scale=0.2, Morning Scale=0.5, Day Scale=0.5, Evening Scale=0.5)
     )
    )
    ScaleA=true
    ScaleRGB=true
   )
  ]
  rendererClasses [&NiagaraDecalRendererProperties]
  renderers [{rendererClass=&NiagaraDecalRendererProperties, rendererIndex=0}]
 }}
summary {
 systemName=NS_Decal
 emitters [{emitterName=GroundDecal, rendererClasses [&NiagaraDecalRendererProperties]}]
 userVariables [
  User.bIsFrontEnd {type=NiagaraBool}
  User.bIsFrontEndPreview {type=NiagaraBool}
 ]
}
