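# NS_DecalSimple is a minimal Niagara example system that projects a decal through a single emitter with a decal renderer.
# An instantaneous burst spawns the particle, whose color, position offset and orientation along an up vector are set at spawn.
# The Decal_Attributes update module drives size, fade and color scaling, with size and RGB built from vector-times-float inputs.
# Particles persist past their lifetime, and the system exposes no user parameters.
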NiagaraSystem /Game/NeverCook/FX/Niagara_FX_Reuse_Examples/Decal/NS_DecalSimple
dependencies {usedDataInterfaces [], usedDynamicInputs [], usedModules [&SystemState]}
emitters {Decal {
  emitterUpdateScript [
   &EmitterState(
    Inactive Response=::NewEnumerator1
    Life Cycle Mode=::NewEnumerator1
    Loop Behavior=::NewEnumerator1
    Loop Duration Mode=::NewEnumerator1
    UseLoopDelay=false
   )
   &SpawnBurst_Instantaneous(Use Spawn Probability=false, UseLoopCountLimit=false)
  ]
  particleSpawnScript [
   &InitializeParticle(Color=<rgba 1 0.2158 0.059 1>, UsePositionOffset=true, Write Material Random=true)
   /Niagara/Modules/Spawn/Decal/InitialDecalOrientation(Orientation Vector:Vector3f=<0 0 1>, Use Rotation Vector=false)
  ]
  particleUpdateScript [
   /Niagara/Modules/Decal/Decal_Attributes(
    Color=1
    Decal Size:Vector3f=&Multiply_VectorByFloat(Vector=<5 5 5>, Float=65)
    Fade=1
    Scale RGB:Vector3f=&Multiply_VectorByFloat(Vector=<1 1 1>, Float=1)
    ScaleA=true
    ScaleRGB=true
    Size=1
   )
   &ParticleState(Kill Particles When Lifetime Has Elapsed=false)
  ]
  rendererClasses [&NiagaraDecalRendererProperties]
  renderers [{rendererClass=&NiagaraDecalRendererProperties, rendererIndex=0}]
 }}
summary {
 systemName=NS_DecalSimple
 emitters [{emitterName=Decal, rendererClasses [&NiagaraDecalRendererProperties]}]
 userVariables []
}
